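# Sound cue for the thrust-start sound of the Cyber Scavenger glider in the battle royale cosmetics content.
# A Random node picks one of three equally weighted thrust wave players, feeding a Modulator that lowers pitch and varies volume slightly on each play.
# It uses glider thrust attenuation and concurrency settings, a glider sound class, and a reduced overall volume.
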SoundCue /BRCosmetics/Sounds/Gliders/Cyber_Scavenger/Glider_Cyber_Scavenger_Thrust_Start_Cue
root=Modulator_0
nodes [
 {id=Modulator_0, children [Random_0], properties {pitchMax=0.8, pitchMin=0.7, volumeMax=1, volumeMin=0.85}}
 {id=Random_0, children [WavePlayer_0,WavePlayer_3,WavePlayer_4], properties {weights [1,1,1]}}
 {id=WavePlayer_0, properties {soundWaveAssetPtr=&/Glider_Cyber_Scavenger_Thrust_01}}
 {id=WavePlayer_3, properties {soundWaveAssetPtr=&/Glider_Cyber_Scavenger_Thrust_03}}
 {id=WavePlayer_4, properties {soundWaveAssetPtr=&/Glider_Cyber_Scavenger_Thrust_04}}
]
properties {
 attenuationSettings=&Fort_Player_Glider_Thrust_Athena_Att
 maxDistance=12500
 soundClassObject=&Fort_Gliders_Land
 volumeMultiplier=0.4
 concurrencySet [&Athena_Glider_Thrust_Conc]
}
